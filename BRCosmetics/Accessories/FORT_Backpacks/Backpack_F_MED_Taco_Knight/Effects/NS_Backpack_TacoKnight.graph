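# Niagara system for the Taco Knight backpack cosmetic that sheds food bits and salsa as the character runs and lands.
# RunningSpeed drives food meshes and salsa drip sprites, muted while skydiving; Landing drives extra food, salsa burst sprites and drool meshes.
# All four emitters spawn on the user-supplied skeletal mesh.
# Drool uses a pendulum constraint, passing pivot and position to its material, with spawn color tinted by time of day.
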NiagaraSystem /BRCosmetics/Accessories/FORT_Backpacks/Backpack_F_MED_Taco_Knight/Effects/NS_Backpack_TacoKnight
dependencies {
 usedDataInterfaces [&NiagaraDataInterfaceCurve, &NiagaraDataInterfaceColorCurve, &NiagaraDataInterfaceVector2DCurve]
 usedDynamicInputs []
 usedModules [&SystemState]
}
emitters {
 FoodBits {
  emitterUpdateScript [
   &EmitterState()
   &SpawnRate(
    Spawn Probability=0.1
    SpawnRate=&MakeCustomFloatFromBool(Bool=$User.IsSkydiving, True Float=0, False Float=3)
    Use Spawn Probability=true
   ) {enabled=false}
   &SpawnRate(
    Spawn Probability=0.25
    SpawnRate=&Multiply_Float(
     A=&MakeCustomFloatFromBool(Bool=$User.IsSkydiving, True Float=0, False Float=0.0007)
     B=$User.RunningSpeed
    )
    Use Spawn Probability=true
   ) {module=SpawnRate003}
   &SpawnRate(SpawnGroup=1, SpawnRate=&Multiply_Float(A=100, B=$User.Landing), Use Spawn Probability=false) {module=SpawnRate002}
   &SpawnRate(SpawnRate=30, Use Spawn Probability=true) {enabled=false, module=SpawnRate001}
  ]
  particleSpawnScript [
   /SetVariables(Particles.MeshIndex=&RandomRangeInt(Minimum=0, Maximum=3)) {isSetParameters=true}
   &InitializeParticle(
    Color Mode=::NewEnumerator0
    Lifetime Max=1
    Lifetime Min=0.5
    Lifetime Mode=::NewEnumerator1
    Mass Max=2
    Mass Mode=::NewEnumerator3
    Mesh Scale=&MakeCustomVectorFromBool(
     Bool=&SetBoolByIntComparison(A=$Particles.MeshIndex, B=0, Comparison Type=::NewEnumerator4)
     True Vector=<1 1 1>
     False Vector=&MakeCustomVectorFromBool(
      Bool=&SetBoolByIntComparison(A=$Particles.MeshIndex, B=1, Comparison Type=::NewEnumerator4)
      True Vector=&RandomRangeVector(Minimum=<0.5 0.5 0.5>, Maximum=<1.5 1.5 1.5>)
      False Vector=&RandomRangeVector(Minimum=<0.1 0.2 0.1>, Maximum=<0.2 0.2 0.2>)
     )
    )
    Mesh Scale Mode=::NewEnumerator4
    Sprite Size Max=<3 14>
    Sprite Size Min=<2 6>
    Sprite Size Mode=::NewEnumerator2
    UsePositionOffset=false
    Write Material Random=true
   )
   &SphereLocation(
    MaskSpawnGroup=false
    Non Uniform Scale=<1 0.25 1>
    Offset=<0 -3 0>
    Override Seed=false
    Sphere Origin:NiagaraPosition=<0 0 0>
    Sphere Radius=17
    Surface Only=false
    UseNonUniformScale=true
    UseOffset=true
    UseSphereOrientation=false
   )
   &SkeletalMeshLocation(
    Bone Sampling Mode=::NewEnumerator4
    Inherit Velocity=false
    MaskSpawnGroup=false
    Sampled Position Offset:NiagaraPosition=$Output.SphereLocation.Position
    Skeletal Mesh=$User.Skeletal Mesh
   )
   &AddVelocityFromPoint(
    Module.Use Falloff Distance=false
    Origin Offset=<0 0 -2>
    Velocity Strength=&MakeCustomFloatFromBool(
     Bool=&SetBoolByIntComparison(A=&SpawnGroup(), B=0, Comparison Type=::NewEnumerator4)
     True Float=5
     False Float=75
    )
   )
   &InitialMeshOrientation(
    Rotation=&RandomRangeVector(Minimum=<0 0 0>, Maximum=<1 1 1>)
    Use Orientation Vector=false
    Use Rotation Vector=true
   )
   &Color(Color=&MakeCustomLinearColorFromBool(
     Bool=&SetBoolByIntComparison(A=$Particles.MeshIndex, B=0, Comparison Type=::NewEnumerator4)
     True Color=&RandomRangeLinearColor(Minimum=<rgba 0 0.2083 0.0834 1>, Maximum=<rgba 1 1 1 1>)
     False Color=&MakeCustomLinearColorFromBool(
      Bool=&SetBoolByIntComparison(A=$Particles.MeshIndex, B=1, Comparison Type=::NewEnumerator4)
      True Color=&RandomRangeLinearColor(Minimum=<rgba 1 0.6405 0.281 1>, Maximum=<rgba 1 1 1 1>, Color Channel Mode=::NewEnumerator2)
      False Color=&RandomRangeLinearColor(Minimum=<rgba 0.3333 0.2872 0.2014 1>, Maximum=<rgba 0.589 0.3147 0.2199 1>, Color Channel Mode=::NewEnumerator2)
     )
    ))
   &Color(Color=&MakeCustomLinearColorFromBool(
     Bool=&SetBoolByIntComparison(A=$Particles.MeshIndex, B=0, Comparison Type=::NewEnumerator4)
     True Color=&RandomRangeLinearColor(Minimum=<rgba 0 0.25 0 1>, Maximum=<rgba 0.684 2 0.0261 1>, Color Channel Mode=::NewEnumerator2)
     False Color=&MakeCustomLinearColorFromBool(
      Bool=&SetBoolByIntComparison(A=$Particles.MeshIndex, B=1, Comparison Type=::NewEnumerator4)
      True Color=&RandomRangeLinearColor(Minimum=<rgba 1 0 0 1>, Maximum=<rgba 2 0.7724 0 1>, Color Channel Mode=::NewEnumerator2)
      False Color=&RandomRangeLinearColor(Minimum=<rgba 0.8125 0.5918 0.0339 1>, Maximum=<rgba 0.3594 0.2548 0.189 1>, Color Channel Mode=::NewEnumerator2)
     )
    )) {enabled=false, module=Color001}
  ]
  particleUpdateScript [
   &ParticleState()
   &GravityForce(Gravity=&MakeCustomVectorFromBool(
     Bool=&SetBoolByIntComparison(A=$Particles.MeshIndex, B=1, Comparison Type=::NewEnumerator6)
     True Vector=<0 0 -800>
     False Vector=<0 0 -400>
    ))
   &Drag(Drag=&RandomRangeFloat(Minimum=1, Maximum=2), Use Linear Drag=true, Use Rotational Drag=false)
   &SolveForcesAndVelocity(Clamp Velocity=false, Limit Acceleration=false)
   &UpdateMeshOrientation(
    Delta Time:NiagaraFloat=&VectorLength(Vector=$Particles.Velocity)
    Rotation Rate=&RandomRangeFloat(Minimum=0.0003, Maximum=0.0005)
    Rotation Vector=&RandomRangeVector(Minimum=<0 0 0>, Maximum=<1 1 1>)
   )
   &ScaleMeshSize(Scale Factor=&VectorFromFloat(Value=&FloatFromCurve(FloatCurve=DataInterface() {propertyValues {Curve {keys [{interpMode=RCIM_Cubic, tangentMode=RCTM_User, time=0.5, value=1}, <at 1 0>]}}})))
  ]
  rendererClasses [&NiagaraMeshRendererProperties]
  renderers [{rendererClass=&NiagaraMeshRendererProperties, rendererIndex=0}]
 }
 RunningSalsaDrips {
  emitterUpdateScript [
   &EmitterState()
   &SpawnRate(
    Spawn Probability=0.25
    SpawnRate=&Multiply_Float(
     A=&MakeCustomFloatFromBool(Bool=$User.IsSkydiving, True Float=0, False Float=0.005)
     B=$User.RunningSpeed
    )
    Use Spawn Probability=true
   ) {module=SpawnRate001}
   &SpawnRate(SpawnRate=100, Use Spawn Probability=false) {enabled=false}
  ]
  particleSpawnScript [
   &InitializeParticle(
    Color=&RandomRangeLinearColor(Minimum=<rgba 1 0.099 0.099 1>, Maximum=<rgba 1 0.9399 0.099 1>)
    Lifetime Max=0.6
    Lifetime Min=0.25
    Lifetime Mode=::NewEnumerator1
    Sprite Size Max=<3 6>
    Sprite Size Min=<1 2>
    Sprite Size Mode=::NewEnumerator2
    UsePositionOffset=false
    Write Material Random=true
   )
   /SetVariables(Particles.Offset:Vector3f=&Add_Vector(A=&RandomVector(Vector Scale=3), B=<0 -6 0>)) {isSetParameters=true}
   &SkeletalMeshLocation(
    Bone Sampling Mode=::NewEnumerator4
    Inherit Velocity=false
    MaskSpawnGroup=false
    Sampled Position Offset=$Particles.Offset
    Sampled Position Offset Coordinate Space:ENiagaraMeshTransforms=::NewEnumerator3
    Skeletal Mesh=$User.Skeletal Mesh
   )
   &InheritVelocity(Module.Limit Speed=true, Velocity Scale=&VectorFromFloat(Value=-0.25)) {module=InheritVelocity001}
   &AddVelocity(Coordinate Space=::World, Velocity=<0 0 -50>)
  ]
  particleUpdateScript [
   &ParticleState()
   &SkeletalMeshLocation(
    Apply To Particle Position=false
    Bone Sampling Mode=::NewEnumerator4
    Inherit Velocity=false
    MaskSpawnGroup=false
    Sampled Position Offset=$Particles.Offset
    Sampled Position Offset Coordinate Space:ENiagaraMeshTransforms=::NewEnumerator3
    Skeletal Mesh=$User.Skeletal Mesh
   ) {enabled=false, module=SkeletalMeshLocation001}
   &GravityForce(Gravity=<0 0 -500>)
   &ScaleSpriteSize(Scale Factor=&Vector2DFromCurve(
     Vector2Curve=DataInterface() {propertyValues {
       XCurve {keys [<at 0.25 1 RCIM_Cubic>, {interpMode=RCIM_Cubic, leaveTangent=-2.9373, time=1, value=0.3}]}
       YCurve {keys [
         {arriveTangent=-1.4216, interpMode=RCIM_Cubic, time=0.0067, value=0.3563}
         {arriveTangent=2.9367, interpMode=RCIM_Cubic, leaveTangent=2.9367, tangentMode=RCTM_User, time=0.2953, value=1.3705}
         {interpMode=RCIM_Cubic, leaveTangent=7.4188, time=1, value=2}
        ]}
      }}
     ScaleCurve=<1 1>
    ))
   &SolveForcesAndVelocity(Clamp Velocity=false, Limit Acceleration=false)
   &DynamicMaterialParameters(
    Index 0 Param 1=0.8
    Index 0 Param 2=&FloatFromCurve(FloatCurve=DataInterface() {propertyValues {Curve {keys [{arriveTangent=-1.8884, interpMode=RCIM_Cubic, leaveTangent=-1.8884, tangentMode=RCTM_User, time=0, value=1}, <at 1 0.26>]}}})
    Index 0 Param 3=1
    Index 0 Param 4=1
    Module.Param0WriteEnabled=true
    Module.Param1WriteEnabled=true
    Module.Param2WriteEnabled=false
    Module.Param3WriteEnabled=false
   )
   &CameraOffset(Camera Offset Amount=&RandomRangeFloat(Minimum=0, Maximum=6))
   &ScaleColor(
    Scale Alpha=&FloatFromCurve(FloatCurve=DataInterface() {propertyValues {Curve {keys [<at 0 1>, {arriveTangent=-1, interpMode=RCIM_Cubic, leaveTangent=-1, time=0.8, value=1}, <at 1 0>]}}})
    ScaleA=true
    ScaleRGB=true
   )
   /SetVariables(Particles.AdjustedPosition:Vector3f=/Niagara/DynamicInputs/Math/Clamp_Vector(
     Vector:Vector3f=Linked() {linkedVariable {name=Particles.Position, type=NiagaraPosition}}
     Min:Vector3f=&Add_Vector(A=$Output.SkeletalMeshLocation001.SampledPosition, B=<-500 -500 -1000>)
     Max:Vector3f=&Add_Vector(A=$Output.SkeletalMeshLocation001.SampledPosition, B=<500 500 10>)
    )) {enabled=false, isSetParameters=true}
   /SetVariables(Particles.AdjustedPosition:Vector3f=&Clamp_Vector(
     Vector:Vector3f=Linked() {linkedVariable {name=Particles.Position, type=NiagaraPosition}}
     Min:Vector3f=&Add_Vector(A=/Niagara/DynamicInputs/Helpers/SimulationPosition(), B=<-500 -500 -1000>)
     Max:Vector3f=&Add_Vector(A=&SimulationPosition(), B=<500 500 10>)
    )) {isSetParameters=true}
  ]
  rendererClasses [&NiagaraSpriteRendererProperties]
  renderers [{rendererClass=&NiagaraSpriteRendererProperties}]
 }
 SalsaBurst {
  emitterUpdateScript [
   &EmitterState()
   &SpawnRate(
    SpawnRate=&MakeCustomFloatFromBool(Bool=&SetBoolByFloatComparison(A=$User.Landing, B=0.7), True Float=40, False Float=0)
    Use Spawn Probability=false
   )
   &SpawnBurst_Instantaneous(Spawn Count=5, Use Spawn Probability=false) {enabled=false}
  ]
  particleSpawnScript [
   &InitializeParticle(
    Color Maximum=<rgba 1 0.7958 0.5625 1>
    Color Minimum=<rgba 1 0.25 0.5125 1>
    Color Mode=::NewEnumerator2
    Lifetime Max=0.3
    Lifetime Min=0.1
    Lifetime Mode=::NewEnumerator1
    Sprite Rotation Mode=::NewEnumerator2
    Sprite Size Mode=::NewEnumerator1
    Uniform Sprite Size Max=32
    Uniform Sprite Size Min=10
    UsePositionOffset=false
    Write Material Random=true
   )
   &SphereLocation(
    MaskSpawnGroup=false
    Offset=<0 -8 0>
    Override Seed=false
    Sphere Origin:NiagaraPosition=<0 0 0>
    Sphere Radius=2
    Surface Only=false
    UseNonUniformScale=false
    UseOffset=true
    UseSphereOrientation=false
   )
   &SkeletalMeshLocation(
    Bone Sampling Mode=::NewEnumerator4
    Inherit Velocity=false
    MaskSpawnGroup=false
    Sampled Position Offset:NiagaraPosition=$Output.SphereLocation.Position
    Skeletal Mesh=$User.Skeletal Mesh
   )
  ]
  particleUpdateScript [
   &ParticleState()
   &DynamicMaterialParameters(
    Index 0 Param 1=$Engine.Time
    Index 0 Param 4=&RandomRangeFloat(Minimum=0, Maximum=1)
    Module.Param0WriteEnabled=true
    Module.Param1WriteEnabled=true
    Module.Param2WriteEnabled=true
    Module.Param3WriteEnabled=true
   )
   &ScaleColor(
    Scale Alpha=&Lerp_Float(A=&RandomRangeFloat(Minimum=0, Maximum=0.25), B=1, Alpha=$Particles.NormalizedAge)
    ScaleA=true
    ScaleRGB=false
   )
   &Color(Color=&ColorFromCurve(ColorCurve=DataInterface() {propertyValues {
       AlphaCurve {keys [<at 0 0 RCIM_Cubic>,<at 1 1 RCIM_Cubic>]}
       BlueCurve {keys [<at 0 0.547 RCIM_Cubic>,<at 0.7464 0.312>]}
       GreenCurve {keys [<at 0 0.547 RCIM_Cubic>,<at 0.7464 0.5528>]}
       RedCurve {keys [<at 0 1 RCIM_Cubic>,<at 0.7464 1>]}
      }})) {enabled=false}
   &ScaleSpriteSize(Scale Factor=&Vector2DFromFloat(Value=&FloatFromCurve(FloatCurve=DataInterface() {propertyValues {Curve {keys [
          {arriveTangent=10.9378, interpMode=RCIM_Cubic, leaveTangent=10.9378, tangentMode=RCTM_User, time=0, value=0}
          {arriveTangent=1.0617, interpMode=RCIM_Cubic, leaveTangent=1.0617, tangentMode=RCTM_User, time=0.0934, value=0.7}
          {arriveTangent=0.218, interpMode=RCIM_Cubic, leaveTangent=0.218, tangentMode=RCTM_User, time=1, value=1}
         ]}}})))
   &CameraOffset(Camera Offset Amount=10)
  ]
  rendererClasses [&NiagaraSpriteRendererProperties]
  renderers [{rendererClass=&NiagaraSpriteRendererProperties}]
 }
 SalsaDrool {
  emitterUpdateScript [
   &EmitterState()
   &SpawnRate(Spawn Probability=25, SpawnRate=&Multiply_Float(A=50, B=$User.Landing), Use Spawn Probability=true)
   &SpawnRate(
    Spawn Probability=25
    SpawnRate=&Multiply_Float(
     A=&MakeCustomFloatFromBool(Bool=$User.IsSkydiving, True Float=0, False Float=0.012)
     B=$User.RunningSpeed
    )
    Use Spawn Probability=true
   ) {enabled=false, module=SpawnRate001}
   &SpawnBurst_Instantaneous(Spawn Count=10, Use Spawn Probability=false) {enabled=false}
  ]
  particleSpawnScript [
   /SetVariables(
    Particles.PendulumDrag=5
    Particles.PendulumLengthScale=0.125
    Particles.PendulumMaskFalloff=1.5
    Particles.PendulumRigidity=0.2
    Particles.PendulumXY:NiagaraFloat=&RandomRangeFloat(Minimum=0.15, Maximum=0.25)
   ) {isSetParameters=true}
   &InitializeParticle(
    Lifetime Max=1
    Lifetime Min=0.5
    Lifetime Mode=::NewEnumerator1
    Mesh Scale=&Multiply_VectorByFloat(
     Vector=&MakeVector(X=$Particles.PendulumXY, Y=$Particles.PendulumXY, Z=$Particles.PendulumLengthScale)
     Float=&RandomRangeFloat(Minimum=0.3, Maximum=0.75)
    )
    Mesh Scale Mode=::NewEnumerator4
    Position Mode=::NewEnumerator1
    Position Offset=&Add_Vector(A=&RandomVector(Vector Scale=3), B=<0 -15 0>)
    Position Offset Coordinate Space=::Local
    UsePositionOffset=true
    Write Material Random=true
   )
   &self:ColorFromTOD(
    Afternoon:LinearColor=<rgba 0.5781 0.5781 0.5781 1>
    Morning:LinearColor=<rgba 0.6024 0.6024 0.6024 1>
    Night:LinearColor=<rgba 0.1119 0.152 0.2378 1>
    TOD:Vector4f=&MakeVector4_FromLinearColor(LinearColor=$NPC.FortniteNPC.FortniteActiveTimeOfDay)
   ) {module=ColorFromTOD}
   /SetVariables(Particles.Offset:NiagaraPosition=$Particles.Position) {isSetParameters=true}
   &SkeletalMeshLocation(
    Bone Sampling Mode=::NewEnumerator4
    Inherit Velocity=false
    MaskSpawnGroup=false
    Sampled Position Offset=$Particles.Offset
    Skeletal Mesh=$User.Skeletal Mesh
   )
  ]
  particleUpdateScript [
   &ParticleState()
   /SetVariables(Particles.MeshOrigin:Vector3f=&Add_Vector(
     A=Linked() {linkedVariable {name=Particles.Initial.Position, type=NiagaraPosition}}
     B=&MakeVector(
      X=0
      Y=0
      Z=&FloatFromCurve(
       FloatCurve=DataInterface() {propertyValues {Curve {keys [
           {arriveTangent=0.1796, interpMode=RCIM_Cubic, leaveTangent=0.1796, leaveTangentWeight=0.1672, tangentMode=RCTM_User, time=0.2, value=0}
           {arriveTangent=1.1359, interpMode=RCIM_Cubic, leaveTangent=1.1359, tangentMode=RCTM_User, time=0.7, value=0.2}
           {arriveTangent=3.2515, interpMode=RCIM_Cubic, leaveTangent=3.2515, tangentMode=RCTM_User, time=1, value=1}
          ]}}}
       Scale Curve=-170
      )
     )
    )) {isSetParameters=true}
   &Drag(Drag=$Particles.PendulumDrag, Use Linear Drag=true, Use Rotational Drag=false)
   &GravityForce(Gravity=<0 0 -900>)
   &PendulumSetup(
    Module.Enable Spring Driver=false
    Pendulum Length:NiagaraFloat=&Multiply_Float(A=100, B=$Particles.PendulumLengthScale)
    Pendulum Pivot:Vector3f=$Particles.MeshOrigin
    "Spring Driven Constraint (Tightness)"=3
   )
   &SolveForcesAndVelocity(Clamp Velocity=false, Limit Acceleration=false)
   &PendulumConstraint()
   &CalculateAccurateVelocity()
   /SetVariables(Particles.MeshFacing:Vector3f=&NormalizeVector(Vector To Normalize=&Subtract_Vector(A=$Particles.MeshOrigin, B=Linked() {linkedVariable {name=Particles.Position, type=NiagaraPosition}}))) {isSetParameters=true}
   /Niagara/Modules/Update/Orientation/V2/OrientMeshToVector(
    Look At Direction:Vector3f=&Multiply_VectorByFloat(Vector=$Particles.MeshFacing, Float=1)
    Override Quat=false
    Reference Axis Vector:Vector3f=<0 0 1>
    UseOptionalSideAxis=false
    UseOptionalUpAxis=false
   )
   &DynamicMaterialParameters(
    Index 2 Param 1=HlslExpression() {hlslExpression=Particles.MeshOrigin.x}
    Index 2 Param 2=HlslExpression() {hlslExpression=Particles.MeshOrigin.y}
    Index 2 Param 3=HlslExpression() {hlslExpression=Particles.MeshOrigin.z}
    Index 2 Param 4=$Particles.PendulumMaskFalloff
    Index 3 Param 1=&MakeFloatFromVector(Vector=Linked() {linkedVariable {name=Particles.Position, type=NiagaraPosition}}, Channel=::NewEnumerator0)
    Index 3 Param 2=&MakeFloatFromVector(Vector=Linked() {linkedVariable {name=Particles.Position, type=NiagaraPosition}}, Channel=::NewEnumerator1)
    Index 3 Param 3=&MakeFloatFromVector(Vector=Linked() {linkedVariable {name=Particles.Position, type=NiagaraPosition}}, Channel=::NewEnumerator2)
    Index 3 Param 4=$Particles.PendulumRigidity
    Module.Param0WriteEnabled2=true
    Module.Param0WriteEnabled3=true
    Module.Param1WriteEnabled2=true
    Module.Param1WriteEnabled3=true
    Module.Param2WriteEnabled2=true
    Module.Param2WriteEnabled3=true
    Module.Param3WriteEnabled2=true
    Module.Param3WriteEnabled3=true
    Write Parameter Index 0=false
    Write Parameter Index 2=true
    Write Parameter Index 3=true
   )
   &ScaleMeshSize(Scale Factor=&MakeVector(
     X=&FloatFromCurve(FloatCurve=DataInterface() {propertyValues {Curve {keys [<at 0 0.2>,<at 0.7 1>]}}})
     Y=&FloatFromCurve(FloatCurve=DataInterface() {propertyValues {Curve {keys [<at 0 0.2>,<at 0.7 1>]}}})
     Z=&FloatFromCurve(FloatCurve=DataInterface() {propertyValues {Curve {keys [
          <at 0 0.25 RCIM_Cubic>
          <at 0.7 0.8055 RCIM_Cubic>
          {arriveTangent=-1.5491, interpMode=RCIM_Cubic, leaveTangent=-1.5491, tangentMode=RCTM_User, time=1, value=0.25}
         ]}}})
    ))
  ]
  rendererClasses [&NiagaraMeshRendererProperties]
  renderers [{rendererClass=&NiagaraMeshRendererProperties, rendererIndex=0}]
 }
}
summary {
 systemName=NS_Backpack_TacoKnight
 emitters [
  {emitterName=FoodBits, rendererClasses [&NiagaraMeshRendererProperties]}
  {emitterName=SalsaBurst, rendererClasses [&NiagaraSpriteRendererProperties]}
  {emitterName=SalsaDrool, rendererClasses [&NiagaraMeshRendererProperties]}
  {emitterName=RunningSalsaDrips, rendererClasses [&NiagaraSpriteRendererProperties]}
 ]
 userVariables [
  User.bIsFrontEnd {type=NiagaraBool}
  User.bIsFrontEndPreview {type=NiagaraBool}
  User.IsSkydiving {type=NiagaraBool}
  User.Landing {type=NiagaraFloat, defaultValue=0, description=""}
  User.RunningSpeed {type=NiagaraFloat, defaultValue=0, description=""}
  User.Skeletal Mesh {
   type=NiagaraDataInterfaceSkeletalMesh
   defaultValue {
    struct=NiagaraExt_VariableValue_DataInterface
    value {dataInterface=&self:NiagaraDataInterfaceSkeletalMesh_0}
   }
  }
 ]
}
