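# Physics asset for the Taco Knight backpack cosmetic, giving its taco-themed dangling parts secondary motion.
# Capsule bodies cover the pack, fork, spoon, straps, salsa, drip, greens and chips, with most constrained directly to the pack bone as a hub.
# Fork and spoon straps chain off their utensils with narrow, mostly single-axis swing.
# Greens favor one swing axis, the salsa drip swings widely, and chips use tiny capsules with near-locked limits.
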PhysicsAsset /BRCosmetics/Accessories/FORT_Backpacks/Backpack_F_MED_Taco_Knight/Meshes/F_MED_Taco_Knight_Pack_Physics
bodies {
 dyn_chips_bot_bk [{shapeType=Capsule, length=0.5, radius=0.5}]
 dyn_chips_lf_mid [{shapeType=Capsule, length=0.5, radius=0.5}]
 dyn_chips_rt_bk [{shapeType=Capsule, length=0.5, radius=0.5}]
 dyn_chips_rt_mid [{shapeType=Capsule, length=0.5, radius=0.5}]
 dyn_chips_top [{shapeType=Capsule, length=0.5, radius=0.5}]
 dyn_chips_top_bk [{shapeType=Capsule, length=0.5, radius=0.5}]
 dyn_chips_top_mid [{shapeType=Capsule, length=0.5, radius=0.5}]
 dyn_fork [{shapeType=Capsule, length=18.3063, radius=3.119, center=<8.4496 0 0>, rotation=<rot 90 0 0>}]
 dyn_fork_strap_01 [{shapeType=Capsule, length=4.8102, radius=1.3741, center=<4 0 0>, rotation=<rot 90 0 0>}]
 dyn_fork_strap_02 [{shapeType=Capsule, length=4.8102, radius=1.3741, center=<4 0 0>, rotation=<rot 90 0 0>}]
 dyn_green_bot [{shapeType=Capsule, length=7.8209, radius=4.0945, center=<4 0 -2>, rotation=<rot 0 0 90>}]
 dyn_green_lt [{shapeType=Capsule, length=7.8209, radius=4.0945, center=<4 0 -2>, rotation=<rot 0 0 90>}]
 dyn_green_rt [{shapeType=Capsule, length=7.8209, radius=4.0945, center=<4 0 -2>, rotation=<rot 0 0 90>}]
 dyn_salsa [{shapeType=Capsule, length=0.1303, radius=1.5646, center=<1.7022 0 0>, rotation=<rot 0 0 90>}]
 dyn_salsa_drip [{shapeType=Capsule, length=2.33, radius=0.575, center=<1.6016 0 0>, rotation=<rot 90 0 0>}]
 dyn_spoon [{shapeType=Capsule, length=17.0736, radius=2.9845, center=<8.9192 0 0>, rotation=<rot 90 0 0>}]
 dyn_spoon_strap [{shapeType=Capsule, length=1.9651, radius=0.505, center=<1 0 0>, rotation=<rot 90 0 0>}]
 pack [{shapeType=Capsule, length=5.3801, radius=3.1674, center=<0 -1.8987 -7.8865>}]
}
constraints [
 {bone1Name=dyn_fork_strap_02, bone2Name=dyn_fork, swing1LimitDegrees=35, swing2LimitDegrees=3, twistLimitDegrees=1}
 {bone1Name=dyn_fork_strap_01, bone2Name=dyn_fork, swing1LimitDegrees=35, swing2LimitDegrees=3, twistLimitDegrees=1}
 {bone1Name=dyn_spoon_strap, bone2Name=dyn_spoon, swing1LimitDegrees=20, swing2LimitDegrees=5, twistLimitDegrees=3}
 {bone1Name=dyn_fork, bone2Name=pack, swing1LimitDegrees=5, swing2LimitDegrees=3, twistLimitDegrees=3}
 {bone1Name=dyn_spoon, bone2Name=pack, swing1LimitDegrees=6, swing2LimitDegrees=3, twistLimitDegrees=3}
 {bone1Name=dyn_salsa, bone2Name=pack, swing1LimitDegrees=20, swing2LimitDegrees=25, twistLimitDegrees=3}
 {bone1Name=dyn_salsa_drip, bone2Name=pack, swing1LimitDegrees=45, swing2LimitDegrees=45, twistLimitDegrees=15}
 {bone1Name=dyn_green_lt, bone2Name=pack, swing1LimitDegrees=3, swing2LimitDegrees=35, twistLimitDegrees=3}
 {bone1Name=dyn_green_bot, bone2Name=pack, swing1LimitDegrees=10, swing2LimitDegrees=45, twistLimitDegrees=5}
 {bone1Name=dyn_green_rt, bone2Name=pack, swing1LimitDegrees=3, swing2LimitDegrees=35, twistLimitDegrees=3}
 {bone1Name=dyn_chips_bot_bk, bone2Name=pack, swing1LimitDegrees=0.1, swing2LimitDegrees=0.1, twistLimitDegrees=0.1}
 {bone1Name=dyn_chips_top, bone2Name=pack, swing1LimitDegrees=0.1, swing2LimitDegrees=0.1, twistLimitDegrees=0.1}
 {bone1Name=dyn_chips_rt_mid, bone2Name=pack, swing1LimitDegrees=0.1, swing2LimitDegrees=0.1, twistLimitDegrees=0.1}
 {bone1Name=dyn_chips_top_mid, bone2Name=pack, swing1LimitDegrees=0.1, swing2LimitDegrees=0.1, twistLimitDegrees=0.1}
 {bone1Name=dyn_chips_lf_mid, bone2Name=pack, swing1LimitDegrees=0.1, swing2LimitDegrees=0.1, twistLimitDegrees=0.1}
 {bone1Name=dyn_chips_rt_bk, bone2Name=pack, swing1LimitDegrees=0.1, swing2LimitDegrees=0.1, twistLimitDegrees=0.1}
 {bone1Name=dyn_chips_top_bk, bone2Name=pack, swing1LimitDegrees=0.1, swing2LimitDegrees=0.1, twistLimitDegrees=0.1}
]
