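# Animation Blueprint for the Taco Knight backpack cosmetic, adding corrective bone motion and physics to the pack.
# Each update maps the owner mesh's spine bend (spine_05 relative to spine_01, from ref pose) to a clamped pack roll.
# The AnimGraph applies that roll and a small offset to the pack bone, then runs a RigidBody simulation in that bone's space.
# Physics alpha drops to half while targeting and returns to full otherwise.
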AnimBlueprint /BRCosmetics/Accessories/FORT_Backpacks/Backpack_F_MED_Taco_Knight/Meshes/F_MED_Taco_Knight_Pack_AnimBP
target_skeleton=&/F_MED_Taco_Knight_Pack_Skeleton
graphs [
 AnimGraph {nodes [
   n1 = "Transform (Modify) Bone - Bone: pack"(ComponentPose=Local To Component(LocalPose=Local Space Ref Pose()), Translation="0,-1,-1.5", Rotation=Get backpack_rotate()) {
    rotationMode=BMM_Additive
    translationMode=BMM_Additive
    alphaBoolBlend {blendOption=Linear}
    boneToModify {boneName=pack}
    translation {y=-1, z=-1.5}
   }
   n2 = RigidBody(ComponentPose=@n1, Alpha=Get target float() -> (target float:double)) {
    bClampLinearTranslationLimitToRefPose=true
    bOverrideWorldGravity=true
    overridePhysicsAsset=&/F_MED_Taco_Knight_Pack_Physics
    simulationSpace=BaseBoneSpace
    worldSpaceMinimumScale=1
    alphaBoolBlend {blendOption=Linear}
    baseBoneRef {boneName=pack}
    componentAppliedLinearAccClamp=<1000 1000 1000>
    componentLinearAccScale=<0.4 0.35 0.45>
    componentLinearVelScale=<0.2 0.2 0.2>
    overrideWorldGravity {z=-980}
   }
   Output Pose(Result=Component To Local(ComponentPose=@n2)) {layerGroup=None}
  ]}
 EventGraph {nodes [
   Event BlueprintUpdateAnimation()
   > n1 = Sequence()
   n2 = GetOwnerSkeletalMeshComponent()
   n3 = BreakRotator(InRot=BreakTransform(InTransform=GetDeltaTransformFromRefPose(self=@n2, BoneName=spine_05, BaseName=spine_01)).Rotation)
   > Set backpack_rotate(backpack_rotate=MakeRotator(Roll=MapRangeClamped(Value=@n3.Yaw, InRangeA=-14, InRangeB=28, OutRangeA=10, OutRangeB=-5)))
   n4 = Branch(execute=@n1.then_1, Condition=Get bIsTargeting())
   > Set target float(target float:double=0.5) -> (Output_Get:double)
   Set target float(execute=@n4.else, target float=1)
  ]}
]
variables [
 backpack_rotate {type=Rotator}
 target float {type=number}
]
